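# Post-process animation Blueprint for the GuardHence male head skeleton, adding physics jiggle to both ears and adjusting the left eyelids.
# Two chained AnimDynamics nodes simulate the ear bones as angular springs in root-relative space, then always-on additive bone modifiers shift and scale the left upper and lower mid-eyelid bones.
# Each update, a nativized state function reads a parameter asset to set the ears' gravity override and store rigid-body settings, which the AnimGraph leaves unused.
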AnimBlueprint /BRCosmetics/Characters/Player/Male/Medium/Heads/M_MED_GuardHence_Head/Meshes/CustomSkeleton/M_MED_GuardHence_Head_AnimBP_PP
target_skeleton=&/Fortnite_Base_Head_GuardHence_Skeleton
graphs [
 AnimGraph {nodes [
   n1 = Get GravityOverride()
   n2 = Anim Dynamics - Bone: dyn_Ear_L!(ComponentPose=Local To Component(LocalPose=Input Pose()), Alpha=1, GravityOverride:Vector=@n1, ExternalForce="0,0,0") -> (Pose) {
    angularDampingOverride=1
    angularSpringConstant=1000
    bAngularSpring=true
    bOverrideAngularDamping=true
    bUseGravityOverride=true
    gravityScale=0.35
    simulationSpace=RootRelative
    boundBone {boneName=dyn_Ear_L}
    componentAppliedLinearAccClamp=<3000 3000 3000>
    componentLinearAccScale=<1 1 1>
    componentLinearVelScale=<1 1 1>
    gravityOverride {z=-980}
    physicsBodyDefinitions [{
      boundBone {boneName=dyn_Ear_L}
      boxExtents=<20 30 30>
      constraintSetup {angularLimitsMax=<0 5 0>, angularLimitsMin=<0 -5 0>, angularTarget=<1 0 0>}
      localJointOffset=<5 0 0>
     }]
   }
   n3 = Anim Dynamics - Bone: dyn_Ear_R!(ComponentPose=@n2, Alpha=1, GravityOverride:Vector=@n1, ExternalForce="0,0,0") -> (Pose) {
    angularDampingOverride=1
    angularSpringConstant=1000
    bAngularSpring=true
    bOverrideAngularDamping=true
    bUseGravityOverride=true
    gravityScale=0.35
    simulationSpace=RootRelative
    boundBone {boneName=dyn_Ear_R}
    componentAppliedLinearAccClamp=<3000 3000 3000>
    componentLinearAccScale=<1 1 1>
    componentLinearVelScale=<1 1 1>
    gravityOverride {z=-980}
    physicsBodyDefinitions [{
      boundBone {boneName=dyn_Ear_R}
      boxExtents=<20 30 30>
      constraintSetup {angularLimitsMax=<0 5 0>, angularLimitsMin=<0 -5 0>, angularTarget=<1 0 0>}
      localJointOffset=<5 0 0>
     }]
   }
   n4 = "Transform (Modify) Bone - Bone: L_eye_lid_upper_mid"!(ComponentPose=@n3, bAlphaBoolEnabled:bool=True, Translation="0,0,-0.2", Rotation="0,0,0", Scale="1.05,1.05,1.05") -> (Pose) {
    alphaInputType=Bool
    rotationMode=BMM_Additive
    rotationSpace=BCS_BoneSpace
    scaleMode=BMM_Additive
    scaleSpace=BCS_BoneSpace
    translationMode=BMM_Additive
    alphaBoolBlend {blendInTime=0.2, blendOutTime=0.2}
    boneToModify {boneName=L_eye_lid_upper_mid}
    scale=<1.05 1.05 1.05>
    translation {z=-0.2}
   }
   n5 = "Transform (Modify) Bone - Bone: L_eye_lid_lower_mid"!(ComponentPose=@n4, bAlphaBoolEnabled:bool=True, Translation="0,0,0.2", Rotation="0,0,0", Scale="1.05,1.05,1.05") -> (Pose) {
    alphaInputType=Bool
    rotationMode=BMM_Additive
    rotationSpace=BCS_BoneSpace
    scaleMode=BMM_Additive
    scaleSpace=BCS_BoneSpace
    translationMode=BMM_Additive
    alphaBoolBlend {blendInTime=0.2, blendOutTime=0.2}
    boneToModify {boneName=L_eye_lid_lower_mid}
    scale=<1.05 1.05 1.05>
    translation {z=0.2}
   }
   Output Pose(Result=Component To Local(ComponentPose=@n5))
  ]}
 EventGraph {nodes [
   Event BlueprintUpdateAnimation()
   > n1 = "Gravity Override State Function (Parameter Asset) (Nativized)"(
    CharacterPartAnimInstance=Self-Reference()
    GameTimeInSeconds=GetGameTimeInSeconds()
    Parameters=&/M_MED_GuardHence_Parameters
   )
   > Set GravityOverride(GravityOverride=@n1)
   > Set RBAN Settings Out(RBAN Settings Out=@n1.OutRigidBodyAnimNodeParameters)
   TryGetPawnOwner()
  ]}
]
variables [
 GravityOverride {type=Vector}
 RBAN Settings Out {type=FortRigidBodyAnimNodeParameters}
]
